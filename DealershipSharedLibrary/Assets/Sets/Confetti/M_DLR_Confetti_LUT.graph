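# Translucent, lit confetti particle material for sprite and Niagara effects, using sub-UV confetti textures for a channel-picked mask and a normal map.
# Particle color is optionally multiplied by a LUT texture sampled at a dynamic-parameter coordinate.
# That color is tinted by time-of-day parameters and also scaled by a dynamic emissive value through a shared emissive-adjustment function.
# The mask is multiplied by particle alpha and a near-camera fade.
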Material /DealershipSharedLibrary/Assets/Sets/Confetti/M_DLR_Confetti_LUT
graph {
 properties {
  bUsedWithNiagaraSprites=true
  bUsedWithParticleSprites=true
  blendMode=BLEND_Translucent
  pixelDepthOffsetMode=PDOM_Legacy
  translucencyLightingMode=TLM_VolumetricPerVertexDirectional
  usedShadingModels=Default Lit
 }
 nodes [
  n1 = DynamicParameter() {defaultValue=<rgba 0 1 1 1>, paramNames [Emissive,LUT,Param3,Param4]}
  n2 = ParticleColor()
  n3 = StaticSwitchParameter(
   True=Multiply(
    A=TextureSampleParameter2D(UVs=@n1.LUT) {parameterName=LUT Texture, samplerType=SAMPLERTYPE_Masks, texture=/Game/Global/Textures/Default/Blanks/T_White}.RGB
    B=@n2.RGB
   )
   False=@n2.RGB
  ) {defaultValue=true, parameterName=Use LUT Texture}
  MaterialFunctionCall(Emissive=Multiply(A=@n3, B=DynamicParameter() {defaultValue=<rgba 0 1 1 1>, paramNames [Emissive,LUT,Param3,Param4]}.Emissive)) {materialFunction=&MF_FX_EmissiveAdjustments}
  Multiply(
   A=MaterialFunctionCall(
    Color_Morning=VectorParameter() {parameterName=ColorMorning, defaultValue=<rgba 0.8 0.8 0.8 0>}.RGB
    Color_Day=VectorParameter() {parameterName=ColorDay, defaultValue=<rgba 1 1 1 0>}.RGB
    Color_Evening=VectorParameter() {parameterName=ColorEvening, defaultValue=<rgba 0.8 0.8 0.8 0>}.RGB
    Color_Night=VectorParameter() {parameterName=ColorNight, defaultValue=<rgba 0.5 0.5 0.5 0>}.RGB
   ) {materialFunction=&Color_Time_of_Day}.Result
   B=@n3
  )
  n4 = TextureSample() {texture=/Game/Effects/Fort_Effects/Textures/Confetti/T_Confetti_SubUV_4X4_A}
  Multiply(
   A=Multiply(A=StaticComponentMaskParameter(None=@n4.RGBA) {defaultB=true, parameterName=PickChannel}, B=ParticleColor().A)
   B=MaterialFunctionCall(Distance=ScalarParameter() {defaultValue=15, parameterName=Nearfade_Distance}) {materialFunction=&Near_Camera_Fade}.Result
  )
  Reroute(None=TextureSample() {samplerType=SAMPLERTYPE_Normal, texture=/Game/Effects/Fort_Effects/Textures/Confetti/T_Confetti_SubUV_4X4_03}.RGB)
 ]
}
